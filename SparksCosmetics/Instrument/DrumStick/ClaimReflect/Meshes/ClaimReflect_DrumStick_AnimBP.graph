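# Animation Blueprint for the ClaimReflect drumstick cosmetic that tunes Chaos cloth simulation on its flag.
# Each update fetches flag and offhand cloth interactors and selects velocity and acceleration clamps by front-end, on-pawn and jumping or falling state.
# Both lookups write the OffHand variable, so the Main_Hand interactor being clamped is never assigned here.
# AnimGraph passes Input Pose straight to output, leaving two Anim Dynamics flag-bone chains disconnected.
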AnimBlueprint /SparksCosmetics/Instrument/DrumStick/ClaimReflect/Meshes/ClaimReflect_DrumStick_AnimBP
target_skeleton=&/ClaimReflect_DrumStick_Skeleton
graphs [
 AnimGraph {nodes [
   n1 = Anim Dynamics - Bone: flag_top_01(Alpha=1, ExternalForce="0,0,0") {
    angularDampingOverride=0.9999
    angularSpringConstant=500
    bAngularSpring=true
    bOverrideAngularDamping=true
    bUseGravityOverride=true
    gravityScale=0.35
    alphaBoolBlend {blendOption=Linear}
    boundBone {boneName=flag_top_01}
    componentAppliedLinearAccClamp=<700 700 700>
    componentLinearAccScale=<0.25 0.25 0.25>
    componentLinearVelScale=<1 1 1>
    gravityOverride {z=-750}
    physicsBodyDefinitions [{
      boundBone {boneName=flag_top_01}
      boxExtents=<5 10 10>
      constraintSetup {angularTargetAxis=AxisY, angularLimitsMax=<5 35 0>, angularLimitsMin=<-5 0 0>, angularTarget=<0 0 0>}
      localJointOffset=<1 -0.1 -0.1>
     }]
    relativeSpaceBone {boneName=handle}
    simSpaceSettings {maxAngularAcceleration=100, maxAngularVelocity=10}
   }
   n2 = Anim Dynamics - Bone: flag_top_02(ComponentPose=@n1, Alpha=1, ExternalForce="0,0,0") {
    angularDampingOverride=0.9999
    angularSpringConstant=500
    bAngularSpring=true
    bOverrideAngularDamping=true
    bUseGravityOverride=true
    gravityScale=0.35
    alphaBoolBlend {blendOption=Linear}
    boundBone {boneName=flag_top_02}
    componentAppliedLinearAccClamp=<700 700 700>
    componentLinearAccScale=<0.25 0.25 0.25>
    componentLinearVelScale=<1 1 1>
    gravityOverride {z=-750}
    physicsBodyDefinitions [{
      boundBone {boneName=flag_top_02}
      boxExtents=<5 10 10>
      constraintSetup {angularLimitsMax=<5 35 0>, angularLimitsMin=<-5 0 0>, angularTarget=<0 0 0>}
      localJointOffset=<1 -0.1 -0.1>
     }]
    relativeSpaceBone {boneName=handle}
    simSpaceSettings {maxAngularAcceleration=100, maxAngularVelocity=10}
   }
   n3 = Anim Dynamics - Bone: flag_top_03(ComponentPose=@n2, Alpha=1, ExternalForce="0,0,0") {
    angularDampingOverride=0.9999
    angularSpringConstant=500
    bAngularSpring=true
    bOverrideAngularDamping=true
    bUseGravityOverride=true
    gravityScale=0.35
    alphaBoolBlend {blendOption=Linear}
    boundBone {boneName=flag_top_03}
    componentAppliedLinearAccClamp=<700 700 700>
    componentLinearAccScale=<0.25 0.25 0.25>
    componentLinearVelScale=<1 1 1>
    gravityOverride {z=-750}
    physicsBodyDefinitions [{
      boundBone {boneName=flag_top_03}
      boxExtents=<5 10 10>
      constraintSetup {angularLimitsMax=<5 35 0>, angularLimitsMin=<-5 0 0>, angularTarget=<0 0 0>}
      localJointOffset=<1 -0.1 -0.1>
     }]
    relativeSpaceBone {boneName=handle}
    simSpaceSettings {maxAngularAcceleration=100, maxAngularVelocity=10}
   }
   n4 = Anim Dynamics - Bone: flag_root(ComponentPose=@n3, Alpha=1, ExternalForce="0,0,0") {
    angularDampingOverride=0.9999
    angularSpringConstant=500
    bAngularSpring=true
    bOverrideAngularDamping=true
    bUseGravityOverride=true
    gravityScale=0.35
    alphaBoolBlend {blendOption=Linear}
    boundBone {boneName=flag_root}
    componentAppliedLinearAccClamp=<700 700 700>
    componentLinearAccScale=<0.25 0.25 0.25>
    componentLinearVelScale=<1 1 1>
    gravityOverride {z=-750}
    physicsBodyDefinitions [{
      boundBone {boneName=flag_root}
      boxExtents=<5 10 10>
      constraintSetup {angularLimitsMax=<5 10 0>, angularLimitsMin=<-5 -5 0>, angularTarget=<1 0 0>}
      localJointOffset=<1 -0.1 -0.1>
     }]
    relativeSpaceBone {boneName=handle}
    simSpaceSettings {maxAngularAcceleration=100, maxAngularVelocity=10}
   }
   n5 = Anim Dynamics - Bone: flag_center_01(ComponentPose=@n4, Alpha=1, ExternalForce="0,0,0") {
    angularDampingOverride=0.9999
    angularSpringConstant=500
    bAngularSpring=true
    bOverrideAngularDamping=true
    bUseGravityOverride=true
    gravityScale=0.35
    alphaBoolBlend {blendOption=Linear}
    boundBone {boneName=flag_center_01}
    componentAppliedLinearAccClamp=<700 700 700>
    componentLinearAccScale=<0.25 0.25 0.25>
    componentLinearVelScale=<1 1 1>
    gravityOverride {z=-750}
    physicsBodyDefinitions [{
      boundBone {boneName=flag_center_01}
      boxExtents=<5 10 10>
      constraintSetup {angularLimitsMax=<5 45 0>, angularLimitsMin=<-5 -25 0>, angularTarget=<1 0 0>}
      localJointOffset=<1 -0.1 -0.1>
     }]
    relativeSpaceBone {boneName=handle}
    simSpaceSettings {maxAngularAcceleration=100, maxAngularVelocity=10}
   }
   n6 = Anim Dynamics - Bone: flag_center_02(ComponentPose=@n5, Alpha=1, ExternalForce="0,0,0") {
    angularDampingOverride=0.9999
    angularSpringConstant=500
    bAngularSpring=true
    bOverrideAngularDamping=true
    bUseGravityOverride=true
    gravityScale=0.35
    alphaBoolBlend {blendOption=Linear}
    boundBone {boneName=flag_center_02}
    componentAppliedLinearAccClamp=<700 700 700>
    componentLinearAccScale=<0.25 0.25 0.25>
    componentLinearVelScale=<1 1 1>
    gravityOverride {z=-750}
    physicsBodyDefinitions [{
      boundBone {boneName=flag_center_02}
      boxExtents=<5 10 10>
      constraintSetup {angularLimitsMax=<5 45 0>, angularLimitsMin=<-5 -25 0>, angularTarget=<1 0 0>}
      localJointOffset=<1 -0.1 -0.1>
     }]
    relativeSpaceBone {boneName=handle}
    simSpaceSettings {maxAngularAcceleration=100, maxAngularVelocity=10}
   }
   n7 = Anim Dynamics - Bone: flag_center_03(ComponentPose=@n6, Alpha=1, ExternalForce="0,0,0") {
    angularDampingOverride=0.9999
    angularSpringConstant=500
    bAngularSpring=true
    bOverrideAngularDamping=true
    bUseGravityOverride=true
    gravityScale=0.35
    alphaBoolBlend {blendOption=Linear}
    boundBone {boneName=flag_center_03}
    componentAppliedLinearAccClamp=<700 700 700>
    componentLinearAccScale=<0.25 0.25 0.25>
    componentLinearVelScale=<1 1 1>
    gravityOverride {z=-750}
    physicsBodyDefinitions [{
      boundBone {boneName=flag_center_03}
      boxExtents=<5 10 10>
      constraintSetup {angularLimitsMax=<5 45 0>, angularLimitsMin=<-5 -25 0>, angularTarget=<1 0 0>}
      localJointOffset=<1 -0.1 -0.1>
     }]
    relativeSpaceBone {boneName=handle}
    simSpaceSettings {maxAngularAcceleration=100, maxAngularVelocity=10}
   }
   n8 = Anim Dynamics - Bone: flag_bottom_01(ComponentPose=@n7, Alpha=1, ExternalForce="0,0,0") {
    angularDampingOverride=0.9999
    angularSpringConstant=500
    bAngularSpring=true
    bOverrideAngularDamping=true
    bUseGravityOverride=true
    gravityScale=0.35
    alphaBoolBlend {blendOption=Linear}
    boundBone {boneName=flag_bottom_01}
    componentAppliedLinearAccClamp=<700 700 700>
    componentLinearAccScale=<0.25 0.25 0.25>
    componentLinearVelScale=<1 1 1>
    gravityOverride {z=-750}
    physicsBodyDefinitions [{
      boundBone {boneName=flag_bottom_01}
      boxExtents=<5 10 10>
      constraintSetup {angularTargetAxis=AxisY, angularLimitsMax=<5 40 0>, angularLimitsMin=<-5 -25 0>, angularTarget=<0 0 0>}
      localJointOffset=<1 -0.1 -0.1>
     }]
    relativeSpaceBone {boneName=handle}
    simSpaceSettings {maxAngularAcceleration=100, maxAngularVelocity=10}
   }
   n9 = Anim Dynamics - Bone: flag_bottom_02(ComponentPose=@n8, Alpha=1, ExternalForce="0,0,0") {
    angularDampingOverride=0.9999
    angularSpringConstant=500
    bAngularSpring=true
    bOverrideAngularDamping=true
    bUseGravityOverride=true
    gravityScale=0.35
    alphaBoolBlend {blendOption=Linear}
    boundBone {boneName=flag_bottom_02}
    componentAppliedLinearAccClamp=<700 700 700>
    componentLinearAccScale=<0.25 0.25 0.25>
    componentLinearVelScale=<1 1 1>
    gravityOverride {z=-750}
    physicsBodyDefinitions [{
      boundBone {boneName=flag_bottom_02}
      boxExtents=<5 10 10>
      constraintSetup {angularLimitsMax=<5 40 0>, angularLimitsMin=<-5 -25 0>, angularTarget=<0 0 0>}
      localJointOffset=<1 -0.1 -0.1>
     }]
    relativeSpaceBone {boneName=handle}
    simSpaceSettings {maxAngularAcceleration=100, maxAngularVelocity=10}
   }
   Anim Dynamics - Bone: flag_bottom_03(ComponentPose=@n9, Alpha=1, ExternalForce="0,0,0") {
    angularDampingOverride=0.9999
    angularSpringConstant=500
    bAngularSpring=true
    bOverrideAngularDamping=true
    bUseGravityOverride=true
    gravityScale=0.35
    alphaBoolBlend {blendOption=Linear}
    boundBone {boneName=flag_bottom_03}
    componentAppliedLinearAccClamp=<700 700 700>
    componentLinearAccScale=<0.25 0.25 0.25>
    componentLinearVelScale=<1 1 1>
    gravityOverride {z=-750}
    physicsBodyDefinitions [{
      boundBone {boneName=flag_bottom_03}
      boxExtents=<5 10 10>
      constraintSetup {angularLimitsMax=<5 40 0>, angularLimitsMin=<-5 -25 0>, angularTarget=<0 0 0>}
      localJointOffset=<1 -0.1 -0.1>
     }]
    relativeSpaceBone {boneName=handle}
    simSpaceSettings {maxAngularAcceleration=100, maxAngularVelocity=10}
   }
   Copy Pose From Mesh() {bUseAttachedParent=true}
   Output Pose(Result=Input Pose())
   Reroute Node(InputPin:Component Space Pose Link Structure=Reroute Node()) -> (OutputPin:Component Space Pose Link Structure)
   n10 = Anim Dynamics - Bone: flag_top_01(Alpha=1, ExternalForce="0,0,0") {
    angularDampingOverride=0.9999
    angularSpringConstant=500
    bAngularSpring=true
    bOverrideAngularDamping=true
    bUseGravityOverride=true
    gravityScale=0.35
    alphaBoolBlend {blendOption=Linear}
    boundBone {boneName=flag_top_01}
    componentAppliedLinearAccClamp=<700 700 700>
    componentLinearAccScale=<0.25 0.25 0.25>
    componentLinearVelScale=<1 1 1>
    gravityOverride {z=-750}
    physicsBodyDefinitions [{
      boundBone {boneName=flag_top_01}
      boxExtents=<5 10 10>
      constraintSetup {angularTargetAxis=AxisY, angularLimitsMax=<5 35 15>, angularLimitsMin=<-5 0 -5>, angularTarget=<0 0 0>}
      localJointOffset=<0.75 0 0>
     }]
    relativeSpaceBone {boneName=handle}
    simSpaceSettings {maxAngularAcceleration=100, maxAngularVelocity=10}
   }
   n11 = Anim Dynamics - Bone: flag_top_02(ComponentPose=@n10, Alpha=1, ExternalForce="0,0,0") {
    angularDampingOverride=0.9999
    angularSpringConstant=500
    bAngularSpring=true
    bOverrideAngularDamping=true
    bUseGravityOverride=true
    gravityScale=0.35
    alphaBoolBlend {blendOption=Linear}
    boundBone {boneName=flag_top_02}
    componentAppliedLinearAccClamp=<700 700 700>
    componentLinearAccScale=<0.25 0.25 0.25>
    componentLinearVelScale=<1 1 1>
    gravityOverride {z=-750}
    physicsBodyDefinitions [{
      boundBone {boneName=flag_top_02}
      boxExtents=<5 10 10>
      constraintSetup {angularLimitsMax=<5 35 15>, angularLimitsMin=<-5 0 -5>, angularTarget=<0 0 0>}
      localJointOffset=<0.75 0 0>
     }]
    relativeSpaceBone {boneName=handle}
    simSpaceSettings {maxAngularAcceleration=100, maxAngularVelocity=10}
   }
   n12 = Anim Dynamics - Bone: flag_top_03(ComponentPose=@n11, Alpha=1, ExternalForce="0,0,0") {
    angularDampingOverride=0.9999
    angularSpringConstant=500
    bAngularSpring=true
    bOverrideAngularDamping=true
    bUseGravityOverride=true
    gravityScale=0.35
    alphaBoolBlend {blendOption=Linear}
    boundBone {boneName=flag_top_03}
    componentAppliedLinearAccClamp=<700 700 700>
    componentLinearAccScale=<0.25 0.25 0.25>
    componentLinearVelScale=<1 1 1>
    gravityOverride {z=-750}
    physicsBodyDefinitions [{
      boundBone {boneName=flag_top_03}
      boxExtents=<5 10 10>
      constraintSetup {angularLimitsMax=<5 35 15>, angularLimitsMin=<-5 0 -5>, angularTarget=<0 0 0>}
      localJointOffset=<0.75 0 0>
     }]
    relativeSpaceBone {boneName=handle}
    simSpaceSettings {maxAngularAcceleration=100, maxAngularVelocity=10}
   }
   n13 = Anim Dynamics - Bone: flag_root(ComponentPose=@n12, Alpha=1, ExternalForce="0,0,0") {
    angularDampingOverride=0.9999
    angularSpringConstant=500
    bAngularSpring=true
    bOverrideAngularDamping=true
    bUseGravityOverride=true
    gravityScale=0.35
    alphaBoolBlend {blendOption=Linear}
    boundBone {boneName=flag_root}
    componentAppliedLinearAccClamp=<700 700 700>
    componentLinearAccScale=<0.25 0.25 0.25>
    componentLinearVelScale=<1 1 1>
    gravityOverride {z=-750}
    physicsBodyDefinitions [{
      boundBone {boneName=flag_root}
      boxExtents=<5 10 10>
      constraintSetup {angularLimitsMax=<5 15 0>, angularLimitsMin=<-5 -10 0>, angularTarget=<1 0 0>}
      localJointOffset=<0.75 0 0>
     }]
    relativeSpaceBone {boneName=handle}
    simSpaceSettings {maxAngularAcceleration=100, maxAngularVelocity=10}
   }
   n14 = Anim Dynamics - Bone: flag_center_01(ComponentPose=@n13, Alpha=1, ExternalForce="0,0,0") {
    angularDampingOverride=0.9999
    angularSpringConstant=500
    bAngularSpring=true
    bOverrideAngularDamping=true
    bUseGravityOverride=true
    gravityScale=0.35
    alphaBoolBlend {blendOption=Linear}
    boundBone {boneName=flag_center_01}
    componentAppliedLinearAccClamp=<700 700 700>
    componentLinearAccScale=<0.25 0.25 0.25>
    componentLinearVelScale=<1 1 1>
    gravityOverride {z=-750}
    physicsBodyDefinitions [{
      boundBone {boneName=flag_center_01}
      boxExtents=<5 10 10>
      constraintSetup {angularLimitsMax=<5 45 15>, angularLimitsMin=<-5 -25 -5>, angularTarget=<1 0 0>}
      localJointOffset=<0.75 0 0>
     }]
    relativeSpaceBone {boneName=handle}
    simSpaceSettings {maxAngularAcceleration=100, maxAngularVelocity=10}
   }
   n15 = Anim Dynamics - Bone: flag_center_02(ComponentPose=@n14, Alpha=1, ExternalForce="0,0,0") {
    angularDampingOverride=0.9999
    angularSpringConstant=500
    bAngularSpring=true
    bOverrideAngularDamping=true
    bUseGravityOverride=true
    gravityScale=0.35
    alphaBoolBlend {blendOption=Linear}
    boundBone {boneName=flag_center_02}
    componentAppliedLinearAccClamp=<700 700 700>
    componentLinearAccScale=<0.25 0.25 0.25>
    componentLinearVelScale=<1 1 1>
    gravityOverride {z=-750}
    physicsBodyDefinitions [{
      boundBone {boneName=flag_center_02}
      boxExtents=<5 10 10>
      constraintSetup {angularLimitsMax=<5 45 15>, angularLimitsMin=<-5 -25 -5>, angularTarget=<1 0 0>}
      localJointOffset=<0.75 0 0>
     }]
    relativeSpaceBone {boneName=handle}
    simSpaceSettings {maxAngularAcceleration=100, maxAngularVelocity=10}
   }
   n16 = Anim Dynamics - Bone: flag_center_03(ComponentPose=@n15, Alpha=1, ExternalForce="0,0,0") {
    angularDampingOverride=0.9999
    angularSpringConstant=500
    bAngularSpring=true
    bOverrideAngularDamping=true
    bUseGravityOverride=true
    gravityScale=0.35
    alphaBoolBlend {blendOption=Linear}
    boundBone {boneName=flag_center_03}
    componentAppliedLinearAccClamp=<700 700 700>
    componentLinearAccScale=<0.25 0.25 0.25>
    componentLinearVelScale=<1 1 1>
    gravityOverride {z=-750}
    physicsBodyDefinitions [{
      boundBone {boneName=flag_center_03}
      boxExtents=<5 10 10>
      constraintSetup {angularLimitsMax=<5 45 15>, angularLimitsMin=<-5 -25 -5>, angularTarget=<1 0 0>}
      localJointOffset=<0.75 0 0>
     }]
    relativeSpaceBone {boneName=handle}
    simSpaceSettings {maxAngularAcceleration=100, maxAngularVelocity=10}
   }
   n17 = Anim Dynamics - Bone: flag_bottom_01(ComponentPose=@n16, Alpha=1, ExternalForce="0,0,0") {
    angularDampingOverride=0.9999
    angularSpringConstant=500
    bAngularSpring=true
    bOverrideAngularDamping=true
    bUseGravityOverride=true
    gravityScale=0.35
    alphaBoolBlend {blendOption=Linear}
    boundBone {boneName=flag_bottom_01}
    componentAppliedLinearAccClamp=<700 700 700>
    componentLinearAccScale=<0.25 0.25 0.25>
    componentLinearVelScale=<1 1 1>
    gravityOverride {z=-750}
    physicsBodyDefinitions [{
      boundBone {boneName=flag_bottom_01}
      boxExtents=<5 10 10>
      constraintSetup {angularTargetAxis=AxisY, angularLimitsMax=<5 40 15>, angularLimitsMin=<-5 -25 -5>, angularTarget=<0 0 0>}
      localJointOffset=<0.75 0 0>
     }]
    relativeSpaceBone {boneName=handle}
    simSpaceSettings {maxAngularAcceleration=100, maxAngularVelocity=10}
   }
   n18 = Anim Dynamics - Bone: flag_bottom_02(ComponentPose=@n17, Alpha=1, ExternalForce="0,0,0") {
    angularDampingOverride=0.9999
    angularSpringConstant=500
    bAngularSpring=true
    bOverrideAngularDamping=true
    bUseGravityOverride=true
    gravityScale=0.35
    alphaBoolBlend {blendOption=Linear}
    boundBone {boneName=flag_bottom_02}
    componentAppliedLinearAccClamp=<700 700 700>
    componentLinearAccScale=<0.25 0.25 0.25>
    componentLinearVelScale=<1 1 1>
    gravityOverride {z=-750}
    physicsBodyDefinitions [{
      boundBone {boneName=flag_bottom_02}
      boxExtents=<5 10 10>
      constraintSetup {angularLimitsMax=<5 40 15>, angularLimitsMin=<-5 -25 -5>, angularTarget=<0 0 0>}
      localJointOffset=<0.75 0 0>
     }]
    relativeSpaceBone {boneName=handle}
    simSpaceSettings {maxAngularAcceleration=100, maxAngularVelocity=10}
   }
   n19 = Anim Dynamics - Bone: flag_bottom_03(ComponentPose=@n18, Alpha=1, ExternalForce="0,0,0") {
    angularDampingOverride=0.9999
    angularSpringConstant=500
    bAngularSpring=true
    bOverrideAngularDamping=true
    bUseGravityOverride=true
    gravityScale=0.35
    alphaBoolBlend {blendOption=Linear}
    boundBone {boneName=flag_bottom_03}
    componentAppliedLinearAccClamp=<700 700 700>
    componentLinearAccScale=<0.25 0.25 0.25>
    componentLinearVelScale=<1 1 1>
    gravityOverride {z=-750}
    physicsBodyDefinitions [{
      boundBone {boneName=flag_bottom_03}
      boxExtents=<5 10 10>
      constraintSetup {angularLimitsMax=<5 40 15>, angularLimitsMin=<-5 -25 -5>, angularTarget=<0 0 0>}
      localJointOffset=<0.75 0 0>
     }]
    relativeSpaceBone {boneName=handle}
    simSpaceSettings {maxAngularAcceleration=100, maxAngularVelocity=10}
   }
   Reroute Node(InputPin=@n19)
  ]}
 EventGraph {nodes [
   n1 = Event BlueprintUpdateAnimation()
   > Set DeltaTimeLocal(DeltaTimeLocal:double=@n1.DeltaTimeX) -> (Output_Get:double)
   > n2 = Sequence!(execute) -> (then_0, then_1, then_2:Exec)
   n3 = GetClothingSimulationInteractor!(self:Skeletal Mesh Component Object Reference=GetPartSkeletalMeshComponent(), ClothingAssetName:Name=None) -> (ReturnValue:Clothing Simulation Interactor Object Reference)
   Is Valid(exec=@n2.then_1, InputObject=@n3)
   n4 = MakeLiteralString(Value=ClaimReflect_DrumStick_Flag_Cloth_0)
   > n5 = Cast To ChaosClothingInteractor!(
    execute
    Object=GetClothingInteractor!(self:Clothing Simulation Interactor Object Reference=@n3, ClothingAssetName:Name=String To Name(InString=@n4)) -> (ReturnValue:Clothing Interactor Object Reference)
   ) -> (then, CastFailed, AsChaos Clothing Interactor:Chaos Clothing Interactor Object Reference) {is=title K2Node_DynamicCast}
   > Set ChaosClothingInteractor_OffHand(ChaosClothingInteractor_OffHand:Chaos Clothing Interactor Object Reference=@n5) -> (Output_Get:Chaos Clothing Interactor Object Reference)
   > n6 = Branch(Condition=AND Boolean(A=Get bIsInFrontEnd(), B=Get bIsOnPawn() -> (bIsOnPawn:bool)))
   > SetVelocityClamps!(
    execute
    self:Chaos Clothing Interactor Object Reference=Get Chaos Clothing Interactor Main_Hand() -> (Chaos Clothing Interactor Main_Hand:Chaos Clothing Interactor Object Reference)
    bEnableLinearVelocityClamping:bool=true
    MaxLinearVelocity:Vector="100,100,100"
    bEnableLinearAccelerationClamping:bool=true
    MaxLinearAcceleration:Vector="200,200,200"
    bEnableAngularVelocityClamping:bool=true
    MaxAngularVelocity:float=1
    bEnableAngularAccelerationClamping:bool=true
    MaxAngularAcceleration:float=2
   ) -> (then)
   n7 = Branch(execute=@n6.else, Condition=Get bIsInFrontEnd())
   n8 = Get Chaos Clothing Interactor Main_Hand()
   > SetVelocityClamps(
    self=@n8
    bEnableLinearVelocityClamping=true
    MaxLinearVelocity="75,75,75"
    bEnableLinearAccelerationClamping=true
    MaxLinearAcceleration="50,50,50"
    bEnableAngularVelocityClamping=true
    MaxAngularVelocity=1
    bEnableAngularAccelerationClamping=true
    MaxAngularAcceleration=2
   )
   n9 = Branch(execute=@n7.else, Condition=OR Boolean(A=Get bIsJumping(), B=Get bIsFalling()))
   > SetVelocityClamps(
    self=@n8
    bEnableLinearVelocityClamping=true
    MaxLinearVelocity="200,200,200"
    bEnableLinearAccelerationClamping=true
    MaxLinearAcceleration="600,600,600"
    bEnableAngularVelocityClamping=true
    MaxAngularVelocity=2
    bEnableAngularAccelerationClamping=true
    MaxAngularAcceleration=4
   )
   SetVelocityClamps(
    execute=@n9.else
    self=@n8
    bEnableLinearVelocityClamping=true
    MaxLinearVelocity="200,200,200"
    bEnableLinearAccelerationClamping=true
    MaxLinearAcceleration="800,800,800"
    bEnableAngularVelocityClamping=true
    MaxAngularVelocity=4
    bEnableAngularAccelerationClamping=true
    MaxAngularAcceleration=8
   )
   n10 = GetClothingSimulationInteractor(self=GetPartSkeletalMeshComponent(), ClothingAssetName=None)
   Is Valid(exec=@n2.then_2, InputObject=@n10)
   n11 = MakeLiteralString(Value=ClaimReflect_DrumStick_Offhand_Clothing_0)
   > n12 = Cast To ChaosClothingInteractor(Object=GetClothingInteractor(self=@n10, ClothingAssetName=String To Name(InString=@n11)))
   > Set ChaosClothingInteractor_OffHand(ChaosClothingInteractor_OffHand=@n12)
   > n13 = Branch(Condition=AND Boolean(A=Get bIsInFrontEnd(), B=Get bIsOnPawn()))
   n14 = Get ChaosClothingInteractor_OffHand() -> (ChaosClothingInteractor_OffHand:Chaos Clothing Interactor Object Reference)
   > SetVelocityClamps(
    self=@n14
    bEnableLinearVelocityClamping=true
    MaxLinearVelocity="100,100,100"
    bEnableLinearAccelerationClamping=true
    MaxLinearAcceleration="200,200,200"
    bEnableAngularVelocityClamping=true
    MaxAngularVelocity=1
    bEnableAngularAccelerationClamping=true
    MaxAngularAcceleration=2
   )
   n15 = Branch(execute=@n13.else, Condition=Get bIsInFrontEnd())
   > SetVelocityClamps(
    self=@n14
    bEnableLinearVelocityClamping=true
    MaxLinearVelocity="75,75,75"
    bEnableLinearAccelerationClamping=true
    MaxLinearAcceleration="50,50,50"
    bEnableAngularVelocityClamping=true
    MaxAngularVelocity=1
    bEnableAngularAccelerationClamping=true
    MaxAngularAcceleration=2
   )
   n16 = Branch(execute=@n15.else, Condition=OR Boolean(A=Get bIsJumping(), B=Get bIsFalling()))
   > SetVelocityClamps(
    self=@n14
    bEnableLinearVelocityClamping=true
    MaxLinearVelocity="200,200,200"
    bEnableLinearAccelerationClamping=true
    MaxLinearAcceleration="600,600,600"
    bEnableAngularVelocityClamping=true
    MaxAngularVelocity=2
    bEnableAngularAccelerationClamping=true
    MaxAngularAcceleration=4
   )
   SetVelocityClamps(
    execute=@n16.else
    self=@n14
    bEnableLinearVelocityClamping=true
    MaxLinearVelocity="200,200,200"
    bEnableLinearAccelerationClamping=true
    MaxLinearAcceleration="800,800,800"
    bEnableAngularVelocityClamping=true
    MaxAngularVelocity=4
    bEnableAngularAccelerationClamping=true
    MaxAngularAcceleration=8
   )
  ]}
]
variables [
 Chaos Clothing Interactor Main_Hand {type=ChaosClothingInteractor, default=null}
 ChaosClothingInteractor_OffHand {type=ChaosClothingInteractor, default=null}
 DeltaTimeLocal {type=number}
 GravityOverride {type=Vector}
]
