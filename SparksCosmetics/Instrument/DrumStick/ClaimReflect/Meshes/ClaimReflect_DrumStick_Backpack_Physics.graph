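# Physics asset for the ClaimReflect drumstick backpack mesh, a Sparks instrument cosmetic, giving it simple collision shapes.
# A single body on the handle bone combines named spheres and thin capsules for the main-hand and off-hand sticks with one large, unnamed capsule offset to the side.
# It has no constraints, so the whole prop behaves as one rigid body with no jointed parts.
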PhysicsAsset /SparksCosmetics/Instrument/DrumStick/ClaimReflect/Meshes/ClaimReflect_DrumStick_Backpack_Physics
bodies {handle [
  {shapeType=Sphere, radius=7.9343, shapeName=offhand_sphere, center=<0 0 33.237>}
  {shapeType=Sphere, radius=7.4346, shapeName=handle_sphere, center=<4.5026 -18.8832 21.9437>}
  {shapeType=Capsule, length=29.4363, radius=3.2593, shapeName=offhand_handle, center=<-1.3281 0.3197 8.3735>, rotation=<rot 0 -90 0>}
  {shapeType=Capsule, length=40.6414, radius=3.1706, shapeName=mainhand_handle, center=<3.26 5.1025 7.9697>, rotation=<rot -60 -90.0002 0.0003>}
  {shapeType=Capsule, length=30.2471, radius=21.7852, center=<-23.8585 2.3106 0.9535>, rotation=<rot -2.4705 1.4265 -30.0308>}
 ]}
constraints []
