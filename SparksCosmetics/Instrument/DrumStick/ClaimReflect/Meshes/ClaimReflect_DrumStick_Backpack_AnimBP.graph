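# Animation Blueprint for the ClaimReflect drumstick back bling, driving physics and cloth motion on its flags.
# The output pose runs only through a handle-relative RigidBody node; Anim Dynamics flag-bone chains are left unconnected.
# Each update sets that node's external force, upward when skydiving with parachute open, otherwise downward.
# It also fetches the MainHand and OffHand Chaos cloth interactors, setting velocity and acceleration clamps by prioritized movement state.
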AnimBlueprint /SparksCosmetics/Instrument/DrumStick/ClaimReflect/Meshes/ClaimReflect_DrumStick_Backpack_AnimBP
target_skeleton=&/ClaimReflect_DrumStick_Backpack_Skeleton
graphs [
 AnimGraph {nodes [
   n1 = Anim Dynamics - Bone: flag_root(Alpha=1, ExternalForce="0,0,0") {
    angularDampingOverride=0.9999
    angularSpringConstant=1500
    bAngularSpring=true
    bOverrideAngularDamping=true
    bUseGravityOverride=true
    gravityScale=0.35
    alphaBoolBlend {blendOption=Linear}
    boundBone {boneName=flag_root}
    componentAppliedLinearAccClamp=<100 100 100>
    gravityOverride {z=-750}
    physicsBodyDefinitions [{
      boundBone {boneName=flag_root}
      boxExtents=<5 10 10>
      constraintSetup {angularLimitsMax=<5 15 0>, angularLimitsMin=<-5 -5 0>, angularTarget=<1 0 0>}
      localJointOffset=<1 0 0>
     }]
    relativeSpaceBone {boneName=handle}
   }
   n2 = Anim Dynamics - Bone: flag_top_01(ComponentPose=@n1, Alpha=1, ExternalForce="0,0,0") {
    angularDampingOverride=0.9999
    angularSpringConstant=10
    bAngularSpring=true
    bOverrideAngularDamping=true
    bUseGravityOverride=true
    gravityScale=0.35
    alphaBoolBlend {blendOption=Linear}
    boundBone {boneName=flag_top_01}
    componentAppliedLinearAccClamp=<100 100 100>
    gravityOverride {z=-750}
    physicsBodyDefinitions [{
      boundBone {boneName=flag_top_01}
      boxExtents=<5 5 5>
      constraintSetup {angularTargetAxis=AxisY, angularLimitsMax=<5 45 15>, angularLimitsMin=<-5 0 -5>, angularTarget=<0 0 0>}
      localJointOffset=<0.1 1 1>
     }]
    relativeSpaceBone {boneName=handle}
   }
   n3 = Anim Dynamics - Bone: flag_top_02(ComponentPose=@n2, Alpha=1, ExternalForce="0,0,0") {
    angularDampingOverride=0.9999
    angularSpringConstant=10
    bAngularSpring=true
    bOverrideAngularDamping=true
    bUseGravityOverride=true
    gravityScale=0.35
    alphaBoolBlend {blendOption=Linear}
    boundBone {boneName=flag_top_02}
    componentAppliedLinearAccClamp=<100 100 100>
    gravityOverride {z=-750}
    physicsBodyDefinitions [{
      boundBone {boneName=flag_top_02}
      boxExtents=<5 5 5>
      constraintSetup {angularLimitsMax=<5 20 15>, angularLimitsMin=<-5 -10 -5>, angularTarget=<0 0 0>}
      localJointOffset=<0.1 1 1>
     }]
    relativeSpaceBone {boneName=handle}
   }
   n4 = Anim Dynamics - Bone: flag_top_03(ComponentPose=@n3, Alpha=1, ExternalForce="0,0,0") {
    angularDampingOverride=0.9999
    angularSpringConstant=10
    bAngularSpring=true
    bOverrideAngularDamping=true
    bUseGravityOverride=true
    gravityScale=0.35
    alphaBoolBlend {blendOption=Linear}
    boundBone {boneName=flag_top_03}
    componentAppliedLinearAccClamp=<100 100 100>
    gravityOverride {z=-750}
    physicsBodyDefinitions [{
      boundBone {boneName=flag_top_03}
      boxExtents=<5 5 5>
      constraintSetup {angularLimitsMax=<5 5 15>, angularLimitsMin=<-5 -5 -5>, angularTarget=<0 0 0>}
      localJointOffset=<0.1 1 1>
     }]
    relativeSpaceBone {boneName=handle}
   }
   n5 = Anim Dynamics - Bone: flag_center_01(ComponentPose=@n4, Alpha=1, ExternalForce="0,0,0") {
    angularDampingOverride=0.9999
    angularSpringConstant=10
    bAngularSpring=true
    bOverrideAngularDamping=true
    bUseGravityOverride=true
    gravityScale=0.35
    alphaBoolBlend {blendOption=Linear}
    boundBone {boneName=flag_center_01}
    componentAppliedLinearAccClamp=<100 100 100>
    gravityOverride {z=-750}
    physicsBodyDefinitions [{
      boundBone {boneName=flag_center_01}
      boxExtents=<5 5 5>
      constraintSetup {angularLimitsMax=<5 60 15>, angularLimitsMin=<-5 -10 -5>, angularTarget=<1 0 0>}
      localJointOffset=<0.1 1 1>
     }]
    relativeSpaceBone {boneName=handle}
   }
   n6 = Anim Dynamics - Bone: flag_center_02(ComponentPose=@n5, Alpha=1, ExternalForce="0,0,0") {
    angularDampingOverride=0.9999
    angularSpringConstant=10
    bAngularSpring=true
    bOverrideAngularDamping=true
    bUseGravityOverride=true
    gravityScale=0.35
    alphaBoolBlend {blendOption=Linear}
    boundBone {boneName=flag_center_02}
    componentAppliedLinearAccClamp=<100 100 100>
    gravityOverride {z=-750}
    physicsBodyDefinitions [{
      boundBone {boneName=flag_center_02}
      boxExtents=<5 5 5>
      constraintSetup {angularLimitsMax=<5 20 15>, angularLimitsMin=<-5 -10 -5>, angularTarget=<1 0 0>}
      localJointOffset=<0.1 1 1>
     }]
    relativeSpaceBone {boneName=handle}
   }
   n7 = Anim Dynamics - Bone: flag_center_03(ComponentPose=@n6, Alpha=1, ExternalForce="0,0,0") {
    angularDampingOverride=0.9999
    angularSpringConstant=10
    bAngularSpring=true
    bOverrideAngularDamping=true
    bUseGravityOverride=true
    gravityScale=0.35
    alphaBoolBlend {blendOption=Linear}
    boundBone {boneName=flag_center_03}
    componentAppliedLinearAccClamp=<100 100 100>
    gravityOverride {z=-750}
    physicsBodyDefinitions [{
      boundBone {boneName=flag_center_03}
      boxExtents=<5 5 5>
      constraintSetup {angularLimitsMax=<5 5 15>, angularLimitsMin=<-5 -5 -5>, angularTarget=<1 0 0>}
      localJointOffset=<0.1 1 1>
     }]
    relativeSpaceBone {boneName=handle}
   }
   n8 = Anim Dynamics - Bone: flag_bottom_01(ComponentPose=@n7, Alpha=1, ExternalForce="0,0,0") {
    angularDampingOverride=0.9999
    angularSpringConstant=10
    bAngularSpring=true
    bOverrideAngularDamping=true
    bUseGravityOverride=true
    gravityScale=0.35
    alphaBoolBlend {blendOption=Linear}
    boundBone {boneName=flag_bottom_01}
    componentAppliedLinearAccClamp=<100 100 100>
    gravityOverride {z=-750}
    physicsBodyDefinitions [{
      boundBone {boneName=flag_bottom_01}
      boxExtents=<5 5 5>
      constraintSetup {angularTargetAxis=AxisY, angularLimitsMax=<5 45 15>, angularLimitsMin=<-5 -25 -5>, angularTarget=<0 0 0>}
      localJointOffset=<0.1 1 1>
     }]
    relativeSpaceBone {boneName=handle}
   }
   n9 = Anim Dynamics - Bone: flag_bottom_02(ComponentPose=@n8, Alpha=1, ExternalForce="0,0,0") {
    angularDampingOverride=0.9999
    angularSpringConstant=10
    bAngularSpring=true
    bOverrideAngularDamping=true
    bUseGravityOverride=true
    gravityScale=0.35
    alphaBoolBlend {blendOption=Linear}
    boundBone {boneName=flag_bottom_02}
    componentAppliedLinearAccClamp=<100 100 100>
    gravityOverride {z=-750}
    physicsBodyDefinitions [{
      boundBone {boneName=flag_bottom_02}
      boxExtents=<5 5 5>
      constraintSetup {angularLimitsMax=<5 20 15>, angularLimitsMin=<-5 -10 -5>, angularTarget=<0 0 0>}
      localJointOffset=<0.1 1 1>
     }]
    relativeSpaceBone {boneName=handle}
   }
   Anim Dynamics - Bone: flag_bottom_03(ComponentPose=@n9, Alpha=1, ExternalForce="0,0,0") {
    angularDampingOverride=0.9999
    angularSpringConstant=10
    bAngularSpring=true
    bOverrideAngularDamping=true
    bUseGravityOverride=true
    gravityScale=0.35
    alphaBoolBlend {blendOption=Linear}
    boundBone {boneName=flag_bottom_03}
    componentAppliedLinearAccClamp=<100 100 100>
    gravityOverride {z=-750}
    physicsBodyDefinitions [{
      boundBone {boneName=flag_bottom_03}
      boxExtents=<5 5 5>
      constraintSetup {angularLimitsMax=<5 5 15>, angularLimitsMin=<-5 -5 -5>, angularTarget=<0 0 0>}
      localJointOffset=<0.1 1 1>
     }]
    relativeSpaceBone {boneName=handle}
   }
   n10 = Anim Dynamics - Bone: Mainhand_flag_root(Alpha=1, ExternalForce="0,0,0") {
    angularDampingOverride=0.9999
    angularSpringConstant=500
    bOverrideAngularDamping=true
    bUseGravityOverride=true
    gravityScale=0.35
    simulationSpace=BoneRelative
    alphaBoolBlend {blendOption=Linear}
    boundBone {boneName=Mainhand_flag_root}
    componentAppliedLinearAccClamp=<700 700 700>
    componentLinearAccScale=<0.25 0.25 0.25>
    componentLinearVelScale=<1 1 1>
    gravityOverride {z=-980}
    physicsBodyDefinitions [{
      boundBone {boneName=Mainhand_flag_root}
      boxExtents=<5 2 10>
      constraintSetup {angularLimitsMax=<0 10 0>, angularLimitsMin=<0 -10 -5>, angularTarget=<1 0 0>}
      localJointOffset=<1 0.1 0.1>
     }]
    relativeSpaceBone {boneName=flag_root}
    simSpaceSettings {maxAngularAcceleration=100, maxAngularVelocity=10}
   }
   n11 = Anim Dynamics - Bone: MainHand_flag_top_01(ComponentPose=@n10, Alpha=1, ExternalForce="0,0,0") {
    angularDampingOverride=0.9999
    angularSpringConstant=500
    bOverrideAngularDamping=true
    bUseGravityOverride=true
    gravityScale=0.35
    alphaBoolBlend {blendOption=Linear}
    boundBone {boneName=MainHand_flag_top_01}
    componentAppliedLinearAccClamp=<700 700 700>
    componentLinearAccScale=<0.25 0.25 0.25>
    componentLinearVelScale=<1 1 1>
    gravityOverride {z=-980}
    physicsBodyDefinitions [{
      boundBone {boneName=MainHand_flag_top_01}
      boxExtents=<5 2 5>
      constraintSetup {angularTargetAxis=AxisY, angularLimitsMax=<5 25 15>, angularLimitsMin=<-5 -10 -2>, angularTarget=<0 0 0>}
      localJointOffset=<0.5 0.1 0.1>
     }]
    relativeSpaceBone {boneName=flag_root}
    simSpaceSettings {maxAngularAcceleration=100, maxAngularVelocity=10}
   }
   n12 = Anim Dynamics - Bone: MainHand_flag_top_02(ComponentPose=@n11, Alpha=1, ExternalForce="0,0,0") {
    angularDampingOverride=0.9999
    angularSpringConstant=500
    bOverrideAngularDamping=true
    bUseGravityOverride=true
    gravityScale=0.35
    alphaBoolBlend {blendOption=Linear}
    boundBone {boneName=MainHand_flag_top_02}
    componentAppliedLinearAccClamp=<700 700 700>
    componentLinearAccScale=<0.25 0.25 0.25>
    componentLinearVelScale=<1 1 1>
    gravityOverride {z=-980}
    physicsBodyDefinitions [{
      boundBone {boneName=MainHand_flag_top_02}
      boxExtents=<5 2 5>
      constraintSetup {angularLimitsMax=<5 15 15>, angularLimitsMin=<-5 -10 -2>, angularTarget=<0 0 0>}
      localJointOffset=<1 0.1 0.1>
     }]
    relativeSpaceBone {boneName=flag_root}
    simSpaceSettings {maxAngularAcceleration=100, maxAngularVelocity=10}
   }
   n13 = Anim Dynamics - Bone: MainHand_flag_top_03(ComponentPose=@n12, Alpha=1, ExternalForce="0,0,0") {
    angularDampingOverride=0.9999
    angularSpringConstant=500
    bOverrideAngularDamping=true
    bUseGravityOverride=true
    gravityScale=0.35
    alphaBoolBlend {blendOption=Linear}
    boundBone {boneName=MainHand_flag_top_03}
    componentAppliedLinearAccClamp=<700 700 700>
    componentLinearAccScale=<0.25 0.25 0.25>
    componentLinearVelScale=<1 1 1>
    gravityOverride {z=-980}
    physicsBodyDefinitions [{
      boundBone {boneName=MainHand_flag_top_03}
      boxExtents=<5 2 5>
      constraintSetup {angularLimitsMax=<5 10 15>, angularLimitsMin=<-5 -10 -2>, angularTarget=<0 0 0>}
      localJointOffset=<1 0.1 0.1>
     }]
    relativeSpaceBone {boneName=flag_root}
    simSpaceSettings {maxAngularAcceleration=100, maxAngularVelocity=10}
   }
   n14 = Anim Dynamics - Bone: MainHand_flag_center_01(ComponentPose=@n13, Alpha=1, ExternalForce="0,0,0") {
    angularDampingOverride=0.9999
    angularSpringConstant=500
    bOverrideAngularDamping=true
    bUseGravityOverride=true
    gravityScale=0.35
    alphaBoolBlend {blendOption=Linear}
    boundBone {boneName=MainHand_flag_center_01}
    componentAppliedLinearAccClamp=<700 700 700>
    componentLinearAccScale=<0.25 0.25 0.25>
    componentLinearVelScale=<1 1 1>
    gravityOverride {z=-980}
    physicsBodyDefinitions [{
      boundBone {boneName=MainHand_flag_center_01}
      boxExtents=<5 2 5>
      constraintSetup {angularTargetAxis=AxisY, angularLimitsMax=<5 25 15>, angularLimitsMin=<-5 -10 -2>, angularTarget=<0 0 0>}
      localJointOffset=<0.5 0.1 0.1>
     }]
    relativeSpaceBone {boneName=flag_root}
    simSpaceSettings {maxAngularAcceleration=100, maxAngularVelocity=10}
   }
   n15 = Anim Dynamics - Bone: MainHand_flag_center_02(ComponentPose=@n14, Alpha=1, ExternalForce="0,0,0") {
    angularDampingOverride=0.9999
    angularSpringConstant=500
    bOverrideAngularDamping=true
    bUseGravityOverride=true
    gravityScale=0.35
    alphaBoolBlend {blendOption=Linear}
    boundBone {boneName=MainHand_flag_center_02}
    componentAppliedLinearAccClamp=<700 700 700>
    componentLinearAccScale=<0.25 0.25 0.25>
    componentLinearVelScale=<1 1 1>
    gravityOverride {z=-980}
    physicsBodyDefinitions [{
      boundBone {boneName=MainHand_flag_center_02}
      boxExtents=<5 2 5>
      constraintSetup {angularLimitsMax=<5 15 15>, angularLimitsMin=<-5 -10 -2>, angularTarget=<0 0 0>}
      localJointOffset=<1 0.1 0.1>
     }]
    relativeSpaceBone {boneName=flag_root}
    simSpaceSettings {maxAngularAcceleration=100, maxAngularVelocity=10}
   }
   n16 = Anim Dynamics - Bone: MainHand_flag_center_03(ComponentPose=@n15, Alpha=1, ExternalForce="0,0,0") {
    angularDampingOverride=0.9999
    angularSpringConstant=500
    bOverrideAngularDamping=true
    bUseGravityOverride=true
    gravityScale=0.35
    alphaBoolBlend {blendOption=Linear}
    boundBone {boneName=MainHand_flag_center_03}
    componentAppliedLinearAccClamp=<700 700 700>
    componentLinearAccScale=<0.25 0.25 0.25>
    componentLinearVelScale=<1 1 1>
    gravityOverride {z=-980}
    physicsBodyDefinitions [{
      boundBone {boneName=MainHand_flag_center_03}
      boxExtents=<5 2 5>
      constraintSetup {angularLimitsMax=<5 10 15>, angularLimitsMin=<-5 -10 -2>, angularTarget=<0 0 0>}
      localJointOffset=<1 0.1 0.1>
     }]
    relativeSpaceBone {boneName=flag_root}
    simSpaceSettings {maxAngularAcceleration=100, maxAngularVelocity=10}
   }
   n17 = Anim Dynamics - Bone: MainHand_flag_bottom_01(ComponentPose=@n16, Alpha=1, ExternalForce="0,0,0") {
    angularDampingOverride=0.9999
    angularSpringConstant=500
    bOverrideAngularDamping=true
    bUseGravityOverride=true
    gravityScale=0.35
    alphaBoolBlend {blendOption=Linear}
    boundBone {boneName=MainHand_flag_bottom_01}
    componentAppliedLinearAccClamp=<700 700 700>
    componentLinearAccScale=<0.25 0.25 0.25>
    componentLinearVelScale=<1 1 1>
    gravityOverride {z=-980}
    physicsBodyDefinitions [{
      boundBone {boneName=MainHand_flag_bottom_01}
      boxExtents=<5 2 5>
      constraintSetup {angularTargetAxis=AxisY, angularLimitsMax=<5 20 15>, angularLimitsMin=<-5 -10 -2>, angularTarget=<0 0 0>}
      localJointOffset=<0.5 0.1 0.1>
     }]
    relativeSpaceBone {boneName=flag_root}
    simSpaceSettings {maxAngularAcceleration=100, maxAngularVelocity=10}
   }
   n18 = Anim Dynamics - Bone: MainHand_flag_bottom_02(ComponentPose=@n17, Alpha=1, ExternalForce="0,0,0") {
    angularDampingOverride=0.9999
    angularSpringConstant=500
    bOverrideAngularDamping=true
    bUseGravityOverride=true
    gravityScale=0.35
    alphaBoolBlend {blendOption=Linear}
    boundBone {boneName=MainHand_flag_bottom_02}
    componentAppliedLinearAccClamp=<700 700 700>
    componentLinearAccScale=<0.25 0.25 0.25>
    componentLinearVelScale=<1 1 1>
    gravityOverride {z=-980}
    physicsBodyDefinitions [{
      boundBone {boneName=MainHand_flag_bottom_02}
      boxExtents=<5 2 5>
      constraintSetup {angularLimitsMax=<5 20 15>, angularLimitsMin=<-5 -10 -2>, angularTarget=<0 0 0>}
      localJointOffset=<1 0.1 0.1>
     }]
    relativeSpaceBone {boneName=flag_root}
    simSpaceSettings {maxAngularAcceleration=100, maxAngularVelocity=10}
   }
   Anim Dynamics - Bone: MainHand_flag_bottom_03(ComponentPose=@n18, Alpha=1, ExternalForce="0,0,0") {
    angularDampingOverride=0.9999
    angularSpringConstant=500
    bOverrideAngularDamping=true
    bUseGravityOverride=true
    gravityScale=0.35
    alphaBoolBlend {blendOption=Linear}
    boundBone {boneName=MainHand_flag_bottom_03}
    componentAppliedLinearAccClamp=<700 700 700>
    componentLinearAccScale=<0.25 0.25 0.25>
    componentLinearVelScale=<1 1 1>
    gravityOverride {z=-980}
    physicsBodyDefinitions [{
      boundBone {boneName=MainHand_flag_bottom_03}
      boxExtents=<5 2 5>
      constraintSetup {angularLimitsMax=<5 10 15>, angularLimitsMin=<-5 -10 -2>, angularTarget=<0 0 0>}
      localJointOffset=<1 0.1 0.1>
     }]
    relativeSpaceBone {boneName=flag_root}
    simSpaceSettings {maxAngularAcceleration=100, maxAngularVelocity=10}
   }
   n19 = RigidBody() {
    bOverrideWorldGravity=true
    overridePhysicsAsset=&/ClaimReflect_DrumStick_Backpack_Physics
    simulationSpace=BaseBoneSpace
    alphaBoolBlend {blendOption=Linear}
   }
   Component To Local(ComponentPose=@n19)
   n20 = RigidBody(ComponentPose=Local To Component(LocalPose=Input Pose()), ExternalForce=Get GravityOverride()) {
    bClampLinearTranslationLimitToRefPose=true
    bOverrideWorldGravity=true
    overridePhysicsAsset=&ClaimReflect_DrumStick_Backpack_Physics
    simulationSpace=BaseBoneSpace
    alphaBoolBlend {blendOption=Linear}
    baseBoneRef {boneName=handle}
    componentAppliedLinearAccClamp=<3000 3000 3000>
    componentLinearAccScale {z=0.5}
    componentLinearVelScale {z=0.5}
    overrideWorldGravity {z=-980}
    simSpaceSettings {
     maxAngularAcceleration=500
     maxAngularVelocity=3
     maxLinearAcceleration=700
     maxLinearVelocity=500
     velocityScaleZ=0.2
     worldAlpha=1
     externalLinearDragV=<1 1>
    }
   }
   Output Pose(Result=Component To Local(ComponentPose=@n20))
   Reroute Node(InputPin:Wildcard) -> (OutputPin:Wildcard)
  ]}
 EventGraph {nodes [
   n1 = Event BlueprintUpdateAnimation()
   > Set DeltaTimeLocal(DeltaTimeLocal:double=@n1.DeltaTimeX) -> (Output_Get:double)
   > n2 = Sequence!(execute) -> (then_0, then_1, then_2:Exec, then_3:Exec, then_4:Exec, then_5:Exec)
   > n3 = Branch(Condition=Get bIsSkyDiving())
   > n4 = Branch(Condition=Get bIsParachuteOpened())
   > Set GravityOverride(GravityOverride="0,0,400")
   Set GravityOverride(execute=@n4.else, GravityOverride="0,0,-980")
   Set GravityOverride(execute=@n3.else, GravityOverride="0,0,-980")
   n5 = GetClothingSimulationInteractor!(self:Skeletal Mesh Component Object Reference=GetPartSkeletalMeshComponent(), ClothingAssetName:Name=None) -> (ReturnValue:Clothing Simulation Interactor Object Reference)
   Is Valid(exec=@n2.then_1, InputObject=@n5)
   n6 = MakeLiteralString(Value=ClaimReflect_DrumStick_OffHand_Clothing_0)
   > n7 = Cast To ChaosClothingInteractor!(
    execute
    Object=GetClothingInteractor!(self:Clothing Simulation Interactor Object Reference=@n5, ClothingAssetName:Name=String To Name(InString=@n6)) -> (ReturnValue:Clothing Interactor Object Reference)
   ) -> (then, CastFailed, AsChaos Clothing Interactor:Chaos Clothing Interactor Object Reference) {is=title K2Node_DynamicCast}
   > Set ChaosClothingInteractor OffHand(ChaosClothingInteractor OffHand:Chaos Clothing Interactor Object Reference=@n7) -> (Output_Get:Chaos Clothing Interactor Object Reference)
   > n8 = Branch(Condition=Get bIsInFrontEnd())
   > SetVelocityClamps!(
    execute
    self:Chaos Clothing Interactor Object Reference=Get ChaosClothingInteractor OffHand() -> (ChaosClothingInteractor OffHand:Chaos Clothing Interactor Object Reference)
    bEnableLinearVelocityClamping:bool=true
    MaxLinearVelocity:Vector="100,100,100"
    bEnableLinearAccelerationClamping:bool=true
    MaxLinearAcceleration:Vector="100,100,100"
    bEnableAngularVelocityClamping:bool=true
    MaxAngularVelocity:float=2
    bEnableAngularAccelerationClamping:bool=true
    MaxAngularAcceleration:float=2
   ) -> (then)
   n9 = Branch(execute=@n8.else, Condition=Get bIsParachuteOpened())
   n10 = Get ChaosClothingInteractor OffHand()
   > SetVelocityClamps(
    self=@n10
    bEnableLinearVelocityClamping=true
    MaxLinearVelocity="225,225,225"
    bEnableLinearAccelerationClamping=true
    MaxLinearAcceleration="400,400,400"
    bEnableAngularVelocityClamping=true
    MaxAngularVelocity=8
    bEnableAngularAccelerationClamping=true
    MaxAngularAcceleration=6
   )
   n11 = Branch(execute=@n9.else, Condition=Get bIsSkydiveDiveMode())
   > SetVelocityClamps(
    self=@n10
    bEnableLinearVelocityClamping=true
    MaxLinearVelocity="400,400,400"
    bEnableLinearAccelerationClamping=true
    MaxLinearAcceleration="650,650,650"
    bEnableAngularVelocityClamping=true
    MaxAngularVelocity=2
    bEnableAngularAccelerationClamping=true
    MaxAngularAcceleration=2
   )
   n12 = Branch(execute=@n11.else, Condition=Get bIsSkydiveFloating())
   > SetVelocityClamps(
    self=Get ChaosClothingInteractor OffHand()
    bEnableLinearVelocityClamping=true
    MaxLinearVelocity="250,250,250"
    bEnableLinearAccelerationClamping=true
    MaxLinearAcceleration="200,200,200"
    bEnableAngularVelocityClamping=true
    MaxAngularVelocity=2
    bEnableAngularAccelerationClamping=true
    MaxAngularAcceleration=2
   )
   n13 = Branch(execute=@n12.else, Condition=Get bIsTacticalSprinting())
   n14 = Get ChaosClothingInteractor OffHand()
   > SetVelocityClamps(
    self=@n14
    bEnableLinearVelocityClamping=true
    MaxLinearVelocity="200,200,200"
    bEnableLinearAccelerationClamping=true
    MaxLinearAcceleration="400,400,400"
    bEnableAngularVelocityClamping=true
    MaxAngularVelocity=2
    bEnableAngularAccelerationClamping=true
    MaxAngularAcceleration=2
   )
   n15 = Branch(execute=@n13.else, Condition=Get bIsSprinting())
   > SetVelocityClamps(
    self=@n14
    bEnableLinearVelocityClamping=true
    MaxLinearVelocity="100,100,100"
    bEnableLinearAccelerationClamping=true
    MaxLinearAcceleration="300,300,300"
    bEnableAngularVelocityClamping=true
    MaxAngularVelocity=3
    bEnableAngularAccelerationClamping=true
    MaxAngularAcceleration=4
   )
   n16 = Branch(execute=@n15.else, Condition=Get bIsBackpedaling() -> (bIsBackpedaling:bool))
   > SetVelocityClamps(
    self=@n14
    bEnableLinearVelocityClamping=true
    MaxLinearVelocity="10,10,10"
    bEnableLinearAccelerationClamping=true
    MaxLinearAcceleration="100,100,100"
    bEnableAngularVelocityClamping=true
    MaxAngularVelocity=1
    bEnableAngularAccelerationClamping=true
    MaxAngularAcceleration=2
   )
   n17 = Branch(execute=@n16.else, Condition=Get bIsJumping())
   n18 = Get ChaosClothingInteractor OffHand()
   > SetVelocityClamps(
    self=@n18
    bEnableLinearVelocityClamping=true
    MaxLinearVelocity="600,600,600"
    bEnableLinearAccelerationClamping=true
    MaxLinearAcceleration="400,400,400"
    bEnableAngularVelocityClamping=true
    MaxAngularVelocity=8
    bEnableAngularAccelerationClamping=true
    MaxAngularAcceleration=10
   )
   SetVelocityClamps(
    execute=@n17.else
    self=@n18
    bEnableLinearVelocityClamping=true
    MaxLinearVelocity="90,90,90"
    bEnableLinearAccelerationClamping=true
    MaxLinearAcceleration="300,300,300"
    bEnableAngularVelocityClamping=true
    MaxAngularVelocity=2
    bEnableAngularAccelerationClamping=true
    MaxAngularAcceleration=4
   )
   n19 = GetClothingSimulationInteractor(self=GetPartSkeletalMeshComponent(), ClothingAssetName=None)
   Is Valid(exec=@n2.then_2, InputObject=@n19)
   n20 = MakeLiteralString(Value=ClaimReflect_DrumStick_MainHand_Clothing_0)
   > n21 = Cast To ChaosClothingInteractor(Object=GetClothingInteractor(self=@n19, ClothingAssetName=String To Name(InString=@n20)))
   > Set Chaos Clothing Interactor_MainHand(Chaos Clothing Interactor_MainHand:Chaos Clothing Interactor Object Reference=@n21) -> (Output_Get:Chaos Clothing Interactor Object Reference)
   > n22 = Branch(Condition=Get bIsInFrontEnd())
   > SetVelocityClamps(
    self=Get Chaos Clothing Interactor_MainHand() -> (Chaos Clothing Interactor_MainHand:Chaos Clothing Interactor Object Reference)
    bEnableLinearVelocityClamping=true
    MaxLinearVelocity="100,100,100"
    bEnableLinearAccelerationClamping=true
    MaxLinearAcceleration="100,100,100"
    bEnableAngularVelocityClamping=true
    MaxAngularVelocity=2
    bEnableAngularAccelerationClamping=true
    MaxAngularAcceleration=2
   )
   n23 = Branch(execute=@n22.else, Condition=Get bIsParachuteOpened())
   n24 = Get Chaos Clothing Interactor_MainHand()
   > SetVelocityClamps(
    self=@n24
    bEnableLinearVelocityClamping=true
    MaxLinearVelocity="160,160,160"
    bEnableLinearAccelerationClamping=true
    MaxLinearAcceleration="400,400,400"
    bEnableAngularVelocityClamping=true
    MaxAngularVelocity=8
    bEnableAngularAccelerationClamping=true
    MaxAngularAcceleration=6
   )
   n25 = Branch(execute=@n23.else, Condition=Get bIsSkydiveDiveMode())
   > SetVelocityClamps(
    self=@n24
    bEnableLinearVelocityClamping=true
    MaxLinearVelocity="400,400,400"
    bEnableLinearAccelerationClamping=true
    MaxLinearAcceleration="650,650,650"
    bEnableAngularVelocityClamping=true
    MaxAngularVelocity=2
    bEnableAngularAccelerationClamping=true
    MaxAngularAcceleration=2
   )
   n26 = Branch(execute=@n25.else, Condition=Get bIsSkydiveFloating())
   > SetVelocityClamps(
    self=Get Chaos Clothing Interactor_MainHand()
    bEnableLinearVelocityClamping=true
    MaxLinearVelocity="250,250,250"
    bEnableLinearAccelerationClamping=true
    MaxLinearAcceleration="200,200,200"
    bEnableAngularVelocityClamping=true
    MaxAngularVelocity=2
    bEnableAngularAccelerationClamping=true
    MaxAngularAcceleration=2
   )
   n27 = Branch(execute=@n26.else, Condition=Get bIsTacticalSprinting())
   n28 = Get Chaos Clothing Interactor_MainHand()
   > SetVelocityClamps(
    self=@n28
    bEnableLinearVelocityClamping=true
    MaxLinearVelocity="200,200,200"
    bEnableLinearAccelerationClamping=true
    MaxLinearAcceleration="400,400,400"
    bEnableAngularVelocityClamping=true
    MaxAngularVelocity=2
    bEnableAngularAccelerationClamping=true
    MaxAngularAcceleration=2
   )
   n29 = Branch(execute=@n27.else, Condition=Get bIsSprinting())
   > SetVelocityClamps(
    self=@n28
    bEnableLinearVelocityClamping=true
    MaxLinearVelocity="100,100,100"
    bEnableLinearAccelerationClamping=true
    MaxLinearAcceleration="300,300,300"
    bEnableAngularVelocityClamping=true
    MaxAngularVelocity=3
    bEnableAngularAccelerationClamping=true
    MaxAngularAcceleration=4
   )
   n30 = Branch(execute=@n29.else, Condition=Get bIsBackpedaling())
   > SetVelocityClamps(
    self=@n28
    bEnableLinearVelocityClamping=true
    MaxLinearVelocity="10,10,10"
    bEnableLinearAccelerationClamping=true
    MaxLinearAcceleration="100,100,100"
    bEnableAngularVelocityClamping=true
    MaxAngularVelocity=1
    bEnableAngularAccelerationClamping=true
    MaxAngularAcceleration=2
   )
   n31 = Branch(execute=@n30.else, Condition=Get bIsJumping())
   n32 = Get Chaos Clothing Interactor_MainHand()
   > SetVelocityClamps(
    self=@n32
    bEnableLinearVelocityClamping=true
    MaxLinearVelocity="600,600,600"
    bEnableLinearAccelerationClamping=true
    MaxLinearAcceleration="400,400,400"
    bEnableAngularVelocityClamping=true
    MaxAngularVelocity=8
    bEnableAngularAccelerationClamping=true
    MaxAngularAcceleration=10
   )
   SetVelocityClamps(
    execute=@n31.else
    self=@n32
    bEnableLinearVelocityClamping=true
    MaxLinearVelocity="150,150,150"
    bEnableLinearAccelerationClamping=true
    MaxLinearAcceleration="200,200,200"
    bEnableAngularVelocityClamping=true
    MaxAngularVelocity=2
    bEnableAngularAccelerationClamping=true
    MaxAngularAcceleration=4
   )
   TryGetPawnOwner()
  ]}
]
variables [
 Chaos Clothing Interactor_MainHand {type=ChaosClothingInteractor, default=null}
 ChaosClothingInteractor OffHand {type=ChaosClothingInteractor, default=null}
 DeltaTimeLocal {type=number}
 GravityOverride {type=Vector}
]
