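# Physics asset that gives the ClaimReflect drum stick cosmetic instrument its collision shape.
# A single body on the handle bone combines a long thin capsule for the shaft with a much wider sphere at the +Z end.
# It has no constraints, so the asset defines no jointed or secondary motion.
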PhysicsAsset /SparksCosmetics/Instrument/DrumStick/ClaimReflect/Meshes/ClaimReflect_DrumStick_Physics
bodies {handle [
  {shapeType=Sphere, radius=7.6149, shapeName=handle_sphere, center=<0 0 32.4957>}
  {shapeType=Capsule, length=43.5011, radius=1.3278, shapeName=thick handle, center=<0 0 3.8692>, rotation=<rot 0 -90 0>}
 ]}
constraints []
